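# Translucent, two-sided test glass parent material, filed under the IcyScarf cosmetic companion, with surface per-pixel lighting and screen-space reflections.
# MF_Glass feeds a thin-film function, both driven by a TechArtMask texture, followed by optional camera dither.
# A switch chooses vertex-hide offset toward object position or character opacity.
# TwoSidedSign then paints front faces red and back faces green, with inverted vertex normals and separate roughness on backs, before TechArt overrides.
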Material /CosmeticCompanions/Assets/Other/IcyScarf/Materials/test/M_GlassParent_TestHql
graph {
 properties {
  bEnableResponsiveAA=true
  bScreenSpaceReflections=true
  bUseMaterialAttributes=true
  bUsedWithClothing=true
  bUsedWithMeshDeformer=true
  bUsedWithMorphTargets=true
  bUsedWithNanite=true
  bUsedWithNiagaraMeshParticles=true
  bUsedWithSkeletalMesh=true
  blendMode=BLEND_Translucent
  ditherOpacityMask=true
  pixelDepthOffsetMode=PDOM_Legacy
  translucencyDirectionalLightingIntensity=5
  translucencyLightingMode=TLM_SurfacePerPixelLighting
  twoSided=true
  usedShadingModels=Default Lit
 }
 nodes [
  n1 = TextureSampleParameter2D() {
   group=00_Glass
   parameterName=TechArtMask
   samplerType=SAMPLERTYPE_Masks
   sortPriority=-96
   texture=&T_BlackMask
  }
  n2 = MaterialFunctionCall!(
   Material Attributes=MaterialFunctionCall!(TechArtMask=@n1.RGB) {materialFunction=/Game/Global/Materials/Character_Parent_2023/Glass_Parent_2023/MF_Glass}.Result
   Fresnel=Fresnel() {baseReflectFraction=0, exponent=1}
   Alpha=@n1.RGB
  ) {materialFunction=/Game/Global/MaterialFunctions/TechArt/MF_ThinFilm_TechArt}
  n3 = StaticSwitchParameter(
   True=MaterialFunctionCall!(In=@n2.Result) {materialFunction=/Game/Packages/Fortress_SharedMaterials/Base_Material_Functions/BaseCustomizationFunctions/MF_FN_Vehicle_CameraDither}.Result
   False=@n2.Result
  ) {group=00_Glass, parameterName=UseCameraDither, sortPriority=250}
  n4 = Subtract(A=ObjectPositionWS(), B=WorldPosition().XYZ)
  n5 = ScalarParameter() {group=00_Glass, parameterName=VertexHide, sortPriority=301}
  n6 = StaticSwitchParameter(
   True=SetMaterialAttributes!(MaterialAttributes=@n3, World Position Offset=Multiply(A=@n4, B=@n5)) {attributeSetTypes [#1]}
   False=MaterialFunctionCall!(Material=@n3) {materialFunction=/Game/Global/Materials/Character_Parent_2023/Functions/BaseFN_CharacterOpacity_2023}.Result
  ) {group=00_Glass, parameterName=UseVertexHide, sortPriority=300}
  n7 = TwoSidedSign()
  n8 = GetMaterialAttributes(None=@n6) {attributeGetTypes [#2, #3]}
  n9 = Multiply(A=VertexNormalWS()) {constB=-1}
  n10 = ScalarParameter() {defaultValue=1, parameterName=arara}
  n11 = Constant3Vector() {constant=<rgba 1 0 0 0>}
  n12 = Constant3Vector() {constant=<rgba 0 1 0.0014 1>}
  MaterialFunctionCall(In=SetMaterialAttributes!(
    MaterialAttributes=@n6
    Normal=If(A=@n7, "A > B"=@n8.MaterialAttributes, "A == B"=@n8.MaterialAttributes, "A < B"=@n9)
    Roughness=If(A=@n7, "A > B"=@n8.Roughness, "A == B"=@n8.Roughness, "A < B"=@n10)
    Base Color=If(A=@n7, "A > B"=@n11, "A == B"=@n11, "A < B"=@n12)
   ) {attributeSetTypes [#2, #3, #4]}) {materialFunction=&MF_TechArtOverrides}
  OneMinus(None=VertexNormalWS())
 ]
}
